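# Sound cue for the Zero Point POI outdoor ambience.
# A looping node wraps a wave player that loops the Zero Point POI ambience wave.
# It uses the shared ambient attenuation and the Athena outdoor ambient sound class.
# A concurrency override limits it to a single playing instance.
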SoundCue /Hera/Audio/POI/ZeroPoint/SC_ZeroPoint_POI_Ambience_Cue
root=Looping_0
nodes [
 {id=Looping_0, children [WavePlayer_0]}
 {id=WavePlayer_0, properties {bLooping=true, soundWaveAssetPtr=&/ZeroPoint_POI_Ambience}}
]
properties {
 attenuationSettings=/AmbientAudioBase/Attenuations/Fort_Ambient_Att
 bOverrideConcurrency=true
 maxDistance=2097152
 soundClassObject=&Fort_Ambient_Outdoors_Athena
 volumeMultiplier=1
 concurrencyOverrides {maxCount=1}
}
