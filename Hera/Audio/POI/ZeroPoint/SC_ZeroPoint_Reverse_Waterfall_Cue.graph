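# Sound cue for the looping reverse waterfall ambience at the Zero Point POI, in the Athena ambient sound class.
# A single looping wave player feeds a Looping node at the root.
# Attenuation is overridden in the cue with logarithmic distance falloff and listener focus that softens the sound when off-focus.
# A concurrency override allows only one instance to play at a time.
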SoundCue /Hera/Audio/POI/ZeroPoint/SC_ZeroPoint_Reverse_Waterfall_Cue
root=Looping_0
nodes [
 {id=Looping_0, children [WavePlayer_0]}
 {id=WavePlayer_0, properties {bLooping=true, soundWaveAssetPtr=&/ZeroPoint_POI_Ambience_Reverse_Waterfall}}
]
properties {
 bOverrideAttenuation=true
 bOverrideConcurrency=true
 maxDistance=6500
 soundClassObject=&Fort_Ambient_Athena
 volumeMultiplier=1
 attenuationOverrides {
  absorptionMethod=CustomCurve
  bEnableFocusInterpolation=true
  bEnableListenerFocus=true
  distanceAlgorithm=Logarithmic
  falloffDistance=5000
  lPFFrequencyAtMin=1000
  lPFRadiusMax=5000
  nonFocusDistanceScale=0.8
  nonFocusVolumeAttenuation=0.6
  occlusionLowPassFilterFrequency=3000
  occlusionTraceChannel=ECC_GameTraceChannel11
  attenuationShapeExtents=<1500 200>
  customAttenuationCurve {editorCurveData {keys [
     <at 0 1 RCIM_Cubic>
     {arriveTangent=-3.3677, interpMode=RCIM_Cubic, leaveTangent=-3.3677, tangentMode=RCTM_User, time=0.1106, value=0.9385}
     {arriveTangent=-1.6386, interpMode=RCIM_Cubic, leaveTangent=-1.6385, tangentMode=RCTM_User, time=0.2529, value=0.5213}
     {arriveTangent=-0.3622, interpMode=RCIM_Cubic, leaveTangent=-0.3622, tangentMode=RCTM_User, time=0.8442, value=0.0744}
     <at 1 0>
    ]}}
  customHighpassAirAbsorptionCurve {editorCurveData {keys [
     <at 0 1 RCIM_Cubic>
     {arriveTangent=-3.3677, interpMode=RCIM_Cubic, leaveTangent=-3.3677, tangentMode=RCTM_User, time=0.0816, value=0.7734}
     {arriveTangent=-1.6386, interpMode=RCIM_Cubic, leaveTangent=-1.6385, tangentMode=RCTM_User, time=0.1318, value=0.4692}
     {arriveTangent=-0.3622, interpMode=RCIM_Cubic, leaveTangent=-0.3622, tangentMode=RCTM_User, time=0.9303, value=0.1245}
     <at 1 0>
    ]}}
  customLowpassAirAbsorptionCurve {editorCurveData {keys [
     <at 0 1 RCIM_Cubic>
     {arriveTangent=-3.3677, interpMode=RCIM_Cubic, leaveTangent=-3.3677, tangentMode=RCTM_User, time=0.0766, value=0.7919}
     {arriveTangent=-1.6386, interpMode=RCIM_Cubic, leaveTangent=-1.6385, tangentMode=RCTM_User, time=0.1448, value=0.4839}
     {arriveTangent=-0.3622, interpMode=RCIM_Cubic, leaveTangent=-0.3622, tangentMode=RCTM_User, time=0.9206, value=0.1465}
     <at 1 0>
    ]}}
 }
 concurrencyOverrides {maxCount=1}
}
